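; Prototype-only quest gameplay ability that spawns test actors, intended as job boards, in front of the player.
; On activation, it clears its tracked array and runs a vertical line trace per spawner class, stepped along world X from the avatar.
; Each hit spawns that class at the ground point pushed along the avatar's forward vector, and records it.
; On ability end, tracked valid actors get a short lifespan so they disappear.
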(event Ability|EventActivateAbility
  (bind _returnvalue (Ability|GetAvatarActorfromActorInfo))
  ;[0 0 0] FOR PROTO ONLY! Spawn job boards in front of the player for testing purposes if they grant the associated test quest
  (|CallParentFunction)
  (Utilities|Array|Clear (Variables|Default|GetSpawnedActors))
  (Variables|Default|SetOffsetFromPrevious 0.0)
  (for _array_element (Variables|Default|GetSpawnerClasses)
    ;;[0.1875 0.1875 0.1875] Spawn actors on the ground in front of the player, offset from each other
    (bind (_outhit _returnvalue_2) (Collision|LineTraceByChannel (Math|Vector|MakeVector (+ (.x (Transformation|GetActorLocation _returnvalue)) (Variables|Default|GetOffsetFromPrevious)) (.y (Transformation|GetActorLocation _returnvalue)) (.z (Transformation|GetActorLocation _returnvalue))) (Math|Vector|vector+vector (Math|Vector|MakeVector (+ (.x (Transformation|GetActorLocation _returnvalue)) (Variables|Default|GetOffsetFromPrevious)) (.y (Transformation|GetActorLocation _returnvalue)) (.z (Transformation|GetActorLocation _returnvalue))) (Math|Vector|MakeVector 0.0 0.0 (Variables|Default|GetDirectlyUnderTraceDownDistance))) (Variables|Default|GetDirectlyUnderTraceChannel)))
    (if _returnvalue_2
      (bind _returnvalue_3 (Game|SpawnActorfromClass _array_element (Math|Transform|MakeTransform (+ (.Location (Collision|BreakHitResult _outhit)) (* (Variables|Default|GetSpawnOffsetFromPlayer) (Transformation|GetActorForwardVector _returnvalue)))) "Undefined" "SelectDefaultAtRuntime" 0 0))
      (Variables|Default|SetOffsetFromPrevious (+ (Variables|Default|GetOffsetFromPrevious) (Variables|Default|GetOffsetPerIteration)))
      (Utilities|Array|AddUnique (Variables|Default|GetSpawnedActors) _returnvalue_3))))

(event Ability|EventOnEndAbility
  (params (bWasCancelled :type Boolean))
  ;[0 0 0] FOR PROTO ONLY! Spawn job boards in front of the player for testing purposes if they grant the associated test quest
  ;;[0.1094 0.1094 0.1094] Remove the actors from the world when player completes the quest
  (|CallParentFunction bWasCancelled)
  (for _array_element (Variables|Default|GetSpawnedActors)
    (Utilities|IsValid _array_element
      (:Is%20Valid
        (Actor|SetLifeSpan _array_element 2.0)))))
